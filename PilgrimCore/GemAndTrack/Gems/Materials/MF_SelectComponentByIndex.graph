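# Material function that selects one component of a four-value vector by a scalar index, stored with the gem materials.
# A Custom HLSL node casts the index to an int, clamps it to 0–3, and indexes the Values vector with it.
# It outputs the selected scalar as Value.
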MaterialFunction /PilgrimCore/GemAndTrack/Gems/Materials/MF_SelectComponentByIndex
graph {nodes [
  MaterialExpressionFunctionInput_0 = FunctionInput() {description=A vector of 4 values, inputName=Values, inputType=FunctionInput_Vector4, previewValue=<0.25 0.5 0.75 1>}
  MaterialExpressionFunctionInput_1 = FunctionInput() {description=An index used to select a value. Clamped 0-3, inputName=Index, inputType=FunctionInput_Scalar, sortPriority=1}
  n1 = Custom!(Values=@MaterialExpressionFunctionInput_0, Index=@MaterialExpressionFunctionInput_1) {
   code="Values[min(max((int)Index, 0), 3)]"
   desc="Covnert to int, clamp to valid values"
   showCode=true
   inputs [
    {inputName=Values, input {expression=&self:MaterialExpressionFunctionInput_0}}
    {inputName=Index, input {expression=&self:MaterialExpressionFunctionInput_1}}
   ]
  }
  FunctionOutput(None=@n1) {description=The scalar value at the provided index from the input vector, outputName=Value}
 ]}
