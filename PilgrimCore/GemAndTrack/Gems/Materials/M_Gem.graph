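# Unlit masked material for note gems on the Pilgrim track, covering kick drum, cymbal, HOPO, pull-off and inverse variants via switches and lerps.
# Custom primitive data supplies lane, track, overdrive and now-bar distance, driving LUT-based color and the track item WPO function.
# A rotated-cubemap fake reflection, UV-masked edges and gradient glows build the emissive, which can darken and desaturate.
# Shared track opacity times per-gem opacity feeds temporal dithering.
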Material /PilgrimCore/GemAndTrack/Gems/Materials/M_Gem
graph {
 properties {
  bEnableResponsiveAA=true
  bOutputTranslucentVelocity=true
  bUsedWithNanite=true
  bUsedWithNiagaraMeshParticles=true
  blendMode=BLEND_Masked
  pixelDepthOffsetMode=PDOM_Legacy
  shadingModel=MSM_Unlit
  usedShadingModels=Unlit
 }
 nodes [
  n1 = Multiply(
   A=MaterialFunctionCall!(
    Values=CollectionParameter() {collection=/SparksEffects/Effects/MPC_Sparks, parameterName=TrackSecondaryElementsPhase}
    Index=NamedRerouteUsage()
   ) {desc=Shared opacity from track, materialFunction=&/MF_SelectComponentByIndex}.Value
   B=NamedRerouteUsage() {desc=Individual gem opacity}
  )
  MaterialFunctionCall(Alpha Threshold=@n1) {materialFunction=&DitherTemporalAA_Fortnite}
  n2 = NamedRerouteUsage()
  n3 = Multiply(A=@n2) {constB=0.1}
  n4 = NamedRerouteUsage()
  MaterialFunctionCall(Emissive=Desaturation(None=LinearInterpolate(A=@n2, B=@n3, Alpha=@n4), Fraction=NamedRerouteUsage())) {materialFunction=&MF_FX_EmissiveAdjustments}
  MaterialFunctionCall!(
   TrackIndex=NamedRerouteUsage()
   UseWorldPositionOffsets=NamedRerouteUsage()
   DistanceFromNowBar=NamedRerouteUsage()
   PreviousFrameDistanceToNowBar=NamedRerouteUsage()
  ) {materialFunction=/PilgrimCore/GemAndTrack/Track/Materials/MF_PilgrimTrackItemWPO}
  Multiply(A=@n1, B=ScalarParameter() {defaultValue=0.5, parameterName=PullOffOpacity})
  n5 = MaterialFunctionCall!(In=VectorParameter() {bUseCustomPrimitiveData=true, parameterName=PilgrimTrackItemStartIndex}.RGBA) {materialFunction=/PilgrimCore/GemAndTrack/Track/Materials/MF_GetPilgrimTrackItemParams}
  NamedRerouteDeclaration(None=@n5.LaneIndex) {name=LaneIndex, nodeColor=<rgba 0 1 0.0291 1>}
  n6 = NamedRerouteUsage()
  n7 = NamedRerouteUsage()
  NamedRerouteDeclaration(None=Clamp(None=Add(A=@n6, B=@n7)) {desc=UseProColors OR IsOverdrive}) {name=ReflectionSaturation, nodeColor=<rgba 0 1 0.0291 1>}
  n8 = NamedRerouteUsage()
  NamedRerouteDeclaration(None=LinearInterpolate(B=Add(A=@n8), Alpha=NamedRerouteUsage()) {desc="Non-pro colors use index 0\r\nPro colors use lane index shifted up by 1, to indices 1-5"}) {name=LaneColorIndex, nodeColor=<rgba 0 1 0.0291 1>}
  NamedRerouteDeclaration(None=@n5.TrackIndex) {name=TrackIndex, nodeColor=<rgba 0 1 0.0291 1>}
  NamedRerouteDeclaration(None=@n5.Last Frame Distance From Now Bar) {name=LastFrameDistanceFromNowBar, nodeColor=<rgba 0 1 0 1>}
  NamedRerouteDeclaration(None=@n5.Distance From Now Bar) {name=DistanceFromNowBar, nodeColor=<rgba 0 1 0 1>}
  NamedRerouteDeclaration(None=@n5.UseWorldPositionOffsets) {name=UseWorldPositionOffsets, nodeColor=<rgba 0 1 0 1>}
  n9 = LinearInterpolate(
   A=Constant3Vector() {constant=<rgba 0 0.2 1 1>}
   B=VectorParameter() {parameterName=KickDrumReflection, defaultValue=<rgba 0.4 1 -0.15 1>}.RGB
   Alpha=ScalarParameter() {parameterName=IsKickDrum}
  ) {constA=0.2, constB=0.3}
  n10 = VectorParameter() {parameterName=CymbalReflection, defaultValue=<rgba 0 -0.2 0.35 1>}
  n11 = ScalarParameter() {parameterName=IsCymbal}
  n12 = ReflectionVectorWS()
  n13 = Add(
   A=RotateAboutAxis(
    NormalizedRotationAxis=LinearInterpolate(A=@n9, B=@n10.RGB, Alpha=@n11)
    RotationAngle=Constant() {r=0.2}
    PivotPoint=Constant()
    Position=@n12
   )
   B=@n12
  )
  n14 = Saturate(None=TextureSampleParameterCube!(UVs=Normalize(VectorInput=@n13), Apply View MipBias) {parameterName=CubeMap, samplerType=SAMPLERTYPE_LinearColor, texture=/PilgrimCore/GemAndTrack/Gems/Textures/T_Pilgrim_Gem_Cube}.RGB)
  n15 = NamedRerouteUsage()
  NamedRerouteDeclaration(None=Power(Base=Desaturation(None=@n14, Fraction=@n15), Exp=Constant() {desc=3, r=3})) {name=Fake Reflection, nodeColor=<rgba 0 0 0 1>}
  n16 = OneMinus(None=TextureCoordinate())
  NamedRerouteDeclaration(None=Step(Y=ComponentMask(None=@n16) {g=true}, X=Constant() {r=0.035}) {constX=0.025, constY=0.025}) {name=Gem Leading Edge, nodeColor=<rgba 0 0 0 1>}
  n17 = OneMinus(None=TextureCoordinate())
  n18 = TextureCoordinate() {coordinateIndex=1}
  n19 = Multiply(
   A=Step(Y=ComponentMask(None=@n17) {g=true}, X=Constant() {r=0.1}) {constX=0.025, constY=0.025}
   B=SmoothStep(Value=MaterialFunctionCall(UV Channel=@n18) {materialFunction=&LinearGradient}.VGradient) {constMax=0.6, constMin=0.9}
  )
  NamedRerouteDeclaration(None=@n19) {name=KickDrum Leading Edge, nodeColor=<rgba 0 0 0 1>}
  n20 = StaticSwitchParameter(True=Fresnel(ExponentIn=Constant() {r=1}) {exponent=2}, False=NamedRerouteUsage()) {parameterName=IsPullOff}
  n21 = NamedRerouteUsage()
  n22 = StaticSwitchParameter(True=Multiply(A=@n21) {constB=0.2}, False=@n21) {parameterName=IsPullOff}
  n23 = NamedRerouteUsage()
  n24 = CurveAtlasRowParameter(CurveTime=Divide(A=Add(A=@n23) {constB=0.5}) {constB=6}) {
   atlas=/PilgrimCore/GemAndTrack/Colors/Sparks_Track_LUT
   curve=/PilgrimCore/GemAndTrack/Colors/Sparks_Track_Gems
   defaultValue=1
   parameterName=GemColor
  }
  n25 = VectorParameter() {parameterName=Overdrive Color, defaultValue=<rgba 1.1698 1.3272 1.75 1>}
  n26 = StaticSwitchParameter(True=Multiply(A=@n25.RGB) {constB=0.5}, False=@n25.RGB) {parameterName=IsPullOff}
  n27 = NamedRerouteUsage()
  n28 = Multiply(A=NamedRerouteDeclaration(None=LinearInterpolate(A=@n24, B=@n26, Alpha=@n27)) {name=Color, nodeColor=<rgba 0 0 0 1>}, B=Constant() {r=5}) {constB=5}
  n29 = Desaturation(None=@n28, Fraction=Constant() {r=-1})
  n30 = Add(
   A=NamedRerouteUsage()
   B=Multiply(A=Multiply(A=@n20, B=@n22), B=StaticSwitchParameter(True=@n29, False=@n28) {parameterName=IsPilgrimInverseGem})
  )
  n31 = Add(A=@n30) {constB=0.01}
  n32 = StaticSwitchParameter(True=Constant() {r=1.5}, False=StaticSwitchParameter(True=Constant() {r=40}, False=Constant() {r=10}) {parameterName=IsPilgrimInverseGem}) {parameterName=IsPullOff}
  NamedRerouteDeclaration(None=LinearInterpolate(A=@n30, B=Multiply(A=@n31, B=@n32) {constB=0.2}, Alpha=NamedRerouteUsage())) {name=Main Gem Visuals, nodeColor=<rgba 0 0 0 1>}
  n33 = Constant() {r=3}
  n34 = NamedRerouteUsage()
  n35 = Multiply(A=@n34, B=@n34)
  n36 = Constant() {r=0.1}
  n37 = SmoothStep(
   Min=Constant() {r=0.8}
   Max=Constant() {r=1}
   Value=MaterialFunctionCall() {materialFunction=&LinearGradient}.VGradient
  )
  n38 = Constant() {r=0.5}
  n39 = Multiply(A=MaterialFunctionCall() {materialFunction=&BoundingBoxBased_0-1_UVW}.R) {constB=0.5}
  n40 = OneMinus(None=Power(Base=Sine(None=@n39), Exp=Constant() {r=0.5}) {constExponent=0.25})
  n41 = Constant() {r=0.05}
  n42 = Multiply(A=Add(A=Add(A=@n19, B=@n36) {constB=0.025}, B=Multiply(A=@n37, B=@n38) {constB=0.5}), B=Saturate(None=Add(A=@n40, B=@n41) {constB=0.2}))
  NamedRerouteDeclaration(None=Multiply(A=Multiply(A=@n33, B=@n35), B=Saturate(None=@n42)) {constA=1}) {name=KickDrumVis, nodeColor=<rgba 0 0 0 1>}
  NamedRerouteDeclaration(None=ScalarParameter() {bUseCustomPrimitiveData=true, parameterName=IsMiss, primitiveDataIndex=5}) {name=IsMiss, nodeColor=<rgba 0 1 0.0291 1>}
  NamedRerouteDeclaration(None=ScalarParameter() {bUseCustomPrimitiveData=true, defaultValue=1, parameterName=Alpha, primitiveDataIndex=4}) {name=Alpha, nodeColor=<rgba 0 1 0.0291 1>}
  NamedRerouteDeclaration(None=@n5.IsOverdrive) {name=IsOverdrive, nodeColor=<rgba 0 1 0.0291 1>}
  n43 = NamedRerouteUsage()
  n44 = NamedRerouteUsage()
  n45 = Multiply(A=ComponentMask(None=TextureCoordinate()) {r=true}) {constB=0.5}
  n46 = Multiply(A=Step(Y=Sine(None=@n45)) {constX=0.8}) {constB=0.175}
  n47 = LinearInterpolate(A=Add(A=Multiply(A=@n43) {constB=0.25}, B=Multiply(A=@n44, B=@n46)), B=NamedRerouteUsage(), Alpha=NamedRerouteUsage())
  n48 = NamedRerouteUsage()
  n49 = Saturate(None=@n48)
  n50 = Saturate(None=Multiply(A=@n48) {constB=3})
  n51 = TextureCoordinate()
  n52 = Step(Y=ComponentMask(None=OneMinus(None=@n51)) {g=true}, X=Constant() {r=0.2}) {constX=0.025, constY=0.025}
  n53 = Saturate(None=OneMinus(None=@n52))
  n54 = Constant() {r=0.25}
  n55 = Multiply(A=@n52, B=OneMinus(None=NamedRerouteUsage()))
  n56 = LinearInterpolate(
   A=NamedRerouteUsage()
   B=LinearInterpolate(A=LinearInterpolate(A=@n49, B=@n50, Alpha=@n53), B=Multiply(A=@n48, B=@n54) {constB=0.1}, Alpha=Saturate(None=@n55))
   Alpha=ScalarParameter() {parameterName=IsCymbal}
  )
  NamedRerouteDeclaration(None=LinearInterpolate(
    A=StaticSwitchParameter(True=@n47, False=@n56) {parameterName=IsHOPO}
    B=NamedRerouteUsage()
    Alpha=ScalarParameter() {parameterName=IsKickDrum}
   )) {name=Final Gem, nodeColor=<rgba 0 0 0 1>}
  n57 = Constant() {r=1}
  n58 = ComponentMask(None=TextureCoordinate()) {r=true}
  n59 = Constant() {r=18}
  n60 = Add(A=Multiply(A=Power(Base=@n58, Exp=@n57), B=Constant() {r=0.5}) {constB=2}, B=Multiply(A=Power(Base=@n58, Exp=@n59)) {constB=2})
  n61 = Power(Base=ComponentMask(None=TextureCoordinate()) {r=true}, Exp=Constant() {r=5})
  n62 = Constant() {r=2}
  n63 = LinearInterpolate(
   A=StaticSwitchParameter(True=Saturate(None=@n60), False=Multiply(A=@n61, B=@n62) {constB=2}) {parameterName=IsPilgrimInverseGem}
   Alpha=ScalarParameter() {parameterName=IsKickDrum}
  ) {constA=0.2, constB=0}
  n64 = Constant() {r=0.175}
  n65 = OneMinus(None=TextureCoordinate())
  n66 = Power(Base=SmoothStep(Min=Subtract(A=@n64) {constB=0.05}, Max=@n64, Value=ComponentMask(None=@n65) {g=true}) {constMax=0.2, constMin=0.1})
  NamedRerouteDeclaration(None=Multiply(A=Multiply(A=@n63, B=@n66), B=NamedRerouteUsage())) {name=Extra Gradient Glow, nodeColor=<rgba 0 0 0 1>}
  NamedRerouteDeclaration(None=ScalarParameter() {bUseCustomPrimitiveData=true, parameterName=Desaturation, primitiveDataIndex=6}) {name=Desaturation, nodeColor=<rgba 0 1 0.0291 1>}
  NamedRerouteDeclaration(None=@n5.UseProColors) {name=UseProColors, nodeColor=<rgba 0 1 0.0291 1>}
 ]
}
