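# Material function that yaw-billboards gem meshes in GemAndTrack toward the camera, outputting a RotatedVector from RotateAboutAxis.
# The direction from the camera's XY to the offset object position becomes a normalized atan2 angle with a quarter-turn shift, minus the mesh's existing Z rotation from FoliageZRotation.
# The rotation runs around the object's orientation axis.
# The scalar Track Position Offset input shifts the object position, pivot and world position alike.
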MaterialFunction /PilgrimCore/GemAndTrack/Track/Materials/MF_GemBillboardingRotation
graph {nodes [
  n1 = NamedRerouteUsage()
  n2 = ObjectPositionWS()
  n3 = CameraPositionWS()
  n4 = Constant3Vector() {constant=<rgba 1 1 0 0>}
  n5 = Normalize(VectorInput=Subtract(A=Add(A=@n1, B=@n2) {constB=0}, B=Multiply(A=@n3, B=@n4)))
  n6 = Arctangent2Fast(Y=ComponentMask(None=@n5) {g=true}, X=ComponentMask(None=@n5) {r=true})
  n7 = Add(B=Divide(A=Divide(A=@n6) {constB=3.1415})) {constA=0.25, constB=0.25}
  n8 = MaterialFunctionCall() {materialFunction=/Engine/ArtTools/RenderToTexture/MaterialFunctions/FoliageZRotation}
  n9 = MaterialFunctionCall() {materialFunction=&ObjectPivotPoint}
  n10 = NamedRerouteUsage()
  n11 = WorldPosition()
  FunctionOutput(None=RotateAboutAxis(
    NormalizedRotationAxis=ObjectOrientation()
    RotationAngle=Subtract(A=@n7, B=@n8."Z Rotation (0-1)")
    PivotPoint=Add(A=@n9.Object Pivot Location, B=@n10)
    Position=Add(A=@n10, B=@n11.XYZ)
   )) {outputName=RotatedVector}
  NamedRerouteDeclaration(None=FunctionInput() {inputName=Track Position Offset, inputType=FunctionInput_Scalar}) {name=Track Position Offset, nodeColor=<rgba 0 1 0.8594 1>}
 ]}
