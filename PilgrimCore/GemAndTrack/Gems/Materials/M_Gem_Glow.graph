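# Translucent unlit glow material for track gems, with a sparkle that tints toward orange during drum overdrive deploy.
# A panning sparkle texture, shaped by a sine gradient, boosts emissive and modulates a radial glow mask with depth fade.
# Opacity multiplies a shared per-track phase from MPC_Sparks by per-gem custom primitive Alpha.
# World position offset combines track-item movement from custom primitive data with gem billboarding.
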Material /PilgrimCore/GemAndTrack/Gems/Materials/M_Gem_Glow
graph {
 properties {bEnableResponsiveAA=true, blendMode=BLEND_Translucent, pixelDepthOffsetMode=PDOM_Legacy, shadingModel=MSM_Unlit, usedShadingModels=Unlit}
 nodes [
  n1 = NamedRerouteUsage()
  Add(
   A=@n1
   B=MaterialFunctionCall!(Track Position Offset=@n1) {materialFunction=/PilgrimCore/GemAndTrack/Track/Materials/MF_GemBillboardingRotation}.RotatedVector
  )
  n2 = LinearInterpolate(
   A=Constant3Vector() {constant=<rgba 1 1 1 1>}
   B=VectorParameter() {parameterName=DrumODDeploy, defaultValue=<rgba 5 1.2525 0 1>}.RGB
   Alpha=ScalarParameter() {parameterName=IsDrumODDeply}
  )
  n3 = TextureCoordinate() {vTiling=0.75}
  n4 = Multiply(A=MaterialFunctionCall() {materialFunction=&LinearGradient}.UGradient) {constB=0.5}
  n5 = Multiply(
   A=TextureSample(UVs=Panner(Coordinate=@n3) {speedY=0.5}) {texture=/PilgrimCore/GemAndTrack/Gems/Textures/T_Sparks_Overdrive_Sparkle}.G
   B=SmoothStep(Value=Sine(None=@n4)) {constMin=0.6}
  )
  n6 = Multiply(A=@n5) {constB=25}
  MaterialFunctionCall(Emissive=Add(A=@n2, B=Multiply(A=@n2, B=@n6))) {materialFunction=&MF_FX_EmissiveAdjustments}
  n7 = Constant() {r=0.4}
  n8 = OneMinus(None=TextureCoordinate())
  n9 = MaterialFunctionCall(
   UVs=MaterialFunctionCall(Texture Scale=Constant2Vector() {g=0.75, r=1.1}) {materialFunction=&ScaleUVsByCenter}.UVs
   Density=Constant() {r=1.5}
  ) {materialFunction=&RadialGradientExponential}
  n10 = Multiply(
   A=SmoothStep(Min=@n7, Max=Add(A=@n7) {constB=0.1}, Value=ComponentMask(None=@n8) {g=true}) {constMax=0.3, constMin=0.2}
   B=Multiply(A=Constant() {r=0.25}, B=Power(Base=@n9.RadialGradientExponential))
  )
  n11 = Multiply(A=@n5, B=@n10)
  Multiply(A=NamedRerouteUsage(), B=DepthFade(Opacity=Add(A=@n11, B=@n10)) {fadeDistanceDefault=5})
  n12 = CollectionParameter() {collection=/SparksEffects/Effects/MPC_Sparks, parameterName=TrackSecondaryElementsPhase}
  n13 = NamedRerouteUsage()
  NamedRerouteDeclaration(None=Multiply(
    A=MaterialFunctionCall!(Values=@n12, Index=@n13) {desc=Shared opacity from track, materialFunction=&/MF_SelectComponentByIndex}.Value
    B=ScalarParameter() {bUseCustomPrimitiveData=true, defaultValue=1, desc=Individual gem opacity, parameterName=Alpha, primitiveDataIndex=4}
   )) {name=Alpha, nodeColor=<rgba 0 1 0.0291 1>}
  n14 = MaterialFunctionCall!(In=VectorParameter() {bUseCustomPrimitiveData=true, parameterName=PilgrimTrackItemStartIndex}.RGBA) {materialFunction=/PilgrimCore/GemAndTrack/Track/Materials/MF_GetPilgrimTrackItemParams}
  NamedRerouteDeclaration(None=MaterialFunctionCall!(
    TrackIndex=NamedRerouteDeclaration(None=@n14.TrackIndex) {name=Track Index, nodeColor=<rgba 0 1 0.0291 1>}
    UseWorldPositionOffsets=@n14.UseWorldPositionOffsets
    DistanceFromNowBar=@n14.UseWorldPositionOffsets
    PreviousFrameDistanceToNowBar=@n14.UseWorldPositionOffsets
   ) {materialFunction=/PilgrimCore/GemAndTrack/Track/Materials/MF_PilgrimTrackItemWPO}.Track Position Offset) {name=Track Position Offset, nodeColor=<rgba 0 1 0.0291 1>}
 ]
}
