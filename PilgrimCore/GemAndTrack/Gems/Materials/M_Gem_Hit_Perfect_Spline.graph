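# Translucent unlit material for Niagara mesh particles, named as the spline effect for a perfect gem hit in PilgrimCore's GemAndTrack.
# A dynamic parameter, SplineThickness, feeds both base and tip widths of SplineThicken, letting particles control line width.
# A sharpened sine falloff across UV0 multiplies an inverted, powered UV1 gradient with a small floor, scaled by particle alpha.
# Responsive AA is enabled.
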Material /PilgrimCore/GemAndTrack/Gems/Materials/M_Gem_Hit_Perfect_Spline
graph {
 properties {
  bEnableResponsiveAA=true
  bUsedWithNiagaraMeshParticles=true
  blendMode=BLEND_Translucent
  pixelDepthOffsetMode=PDOM_Legacy
  previewMesh=/PilgrimCore/GemAndTrack/Gems/Meshes/SM_Smasher_Spline
  shadingModel=MSM_Unlit
  usedShadingModels=Unlit
 }
 nodes [
  n1 = TextureCoordinate()
  n2 = Sine(None=Multiply(A=ComponentMask(None=@n1) {r=true}) {constB=0.5})
  ConstantBiasScale(None=@n2)
  n3 = DynamicParameter() {paramNames [SplineThickness, Param2, Param3, Param4]}
  MaterialFunctionCall(WidthBase=@n3.SplineThickness, WidthTip=@n3.SplineThickness) {materialFunction=&SplineThicken}
  n4 = ComponentMask(None=TextureCoordinate() {coordinateIndex=1}) {g=true}
  n5 = OneMinus(None=@n4)
  n6 = Add(A=Power(Base=ConstantBiasScale(None=@n5)) {constExponent=10}) {constB=0.025}
  Multiply(A=ParticleColor().A, B=Multiply(A=Power(Base=@n2) {constExponent=5}, B=Saturate(None=@n6)))
  Sine(None=Multiply(A=@n4) {constB=0.5})
 ]
}
