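# Translucent unlit material that draws a soft black shadow for gems on the Pilgrim track.
# Per-primitive custom data supplies track item parameters, which drive the shared track-item world position offset function.
# Opacity multiplies a procedural band and a smoothstepped inverted-V gradient by a per-track phase from MPC_Sparks, selected by track index.
# A per-gem custom-data Alpha and an OverallOpacity parameter scale it further.
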Material /PilgrimCore/GemAndTrack/Gems/Materials/M_Gem_Shadow
graph {
 properties {blendMode=BLEND_Translucent, pixelDepthOffsetMode=PDOM_Legacy, shadingModel=MSM_Unlit, usedShadingModels=Unlit}
 nodes [
  n1 = MaterialFunctionCall!(In=VectorParameter() {bUseCustomPrimitiveData=true, parameterName=PilgrimTrackItemStartIndex}.RGBA) {materialFunction=/PilgrimCore/GemAndTrack/Track/Materials/MF_GetPilgrimTrackItemParams}
  MaterialFunctionCall!(
   TrackIndex=@n1.TrackIndex
   UseWorldPositionOffsets=@n1.TrackIndex
   DistanceFromNowBar=@n1.TrackIndex
   PreviousFrameDistanceToNowBar=@n1.TrackIndex
  ) {materialFunction=/PilgrimCore/GemAndTrack/Track/Materials/MF_PilgrimTrackItemWPO}
  n2 = MaterialFunctionCall!(Width=Constant() {r=0.5}, Sharpness=Constant() {r=0.65}, Offset, Direction Switch=StaticBool() {value=true}, Compare, Input Coordinates) {materialFunction=/Engine/Functions/Engine_MaterialFunctions03/Procedurals/GeneratedBand}
  n3 = TextureCoordinate()
  n4 = SmoothStep(
   Min=ScalarParameter() {parameterName=Param_5}
   Max=ScalarParameter() {defaultValue=1, parameterName=Param_6}
   Value=OneMinus(None=ComponentMask(None=@n3) {g=true})
  )
  n5 = MaterialFunctionCall!(Values=CollectionParameter() {collection=/SparksEffects/Effects/MPC_Sparks, parameterName=TrackSecondaryElementsPhase}, Index=@n1.TrackIndex) {desc=Shared opacity from track, materialFunction=&/MF_SelectComponentByIndex}
  n6 = ScalarParameter() {bUseCustomPrimitiveData=true, defaultValue=1, desc=Individual gem opacity, parameterName=Alpha, primitiveDataIndex=4}
  Multiply(
   A=Multiply(A=Multiply(A=@n2.Result, B=@n4), B=Multiply(A=@n5.Value, B=@n6))
   B=ScalarParameter() {defaultValue=1, parameterName=OverallOpacity}
  )
  VectorParameter() {defaultValue=<rgba 0 0 0 1>}
 ]
}
