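# Unlit, two-sided translucent glow material for the gem smash effect, built for Niagara mesh particles.
# Particle color is boosted and run through MF_FX_EmissiveAdjustments for the glow color.
# Opacity combines particle alpha, a sharp power falloff along V, tiled hex noise scrolling quickly in V, and a sine of U that softens the side edges.
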Material /PilgrimCore/GemAndTrack/Gems/Materials/M_Gem_Smash_Glow
graph {
 properties {
  bUsedWithNiagaraMeshParticles=true
  blendMode=BLEND_Translucent
  pixelDepthOffsetMode=PDOM_Legacy
  previewMesh=/PilgrimCore/GemAndTrack/Gems/Meshes/SM_Gem_Smash_Glow
  shadingModel=MSM_Unlit
  twoSided=true
  usedShadingModels=Unlit
 }
 nodes [
  n1 = ParticleColor()
  MaterialFunctionCall(Emissive=Multiply(A=Constant() {r=2}, B=@n1.RGB)) {materialFunction=&MF_FX_EmissiveAdjustments}
  n2 = Power(Base=ComponentMask(None=TextureCoordinate()) {g=true}, Exp=Constant() {r=7})
  n3 = TextureCoordinate()
  n4 = ComponentMask(None=VectorParameter() {parameterName=Tile, defaultValue=<rgba 0.75 0.5 0 1>}.RGB) {g=true, r=true}
  n5 = Time()
  n6 = Constant() {r=3}
  n7 = TextureSample(UVs=Panner(Coordinate=Multiply(A=@n3, B=@n4), Time=Multiply(A=@n5, B=@n6)) {speedY=1}) {texture=/Game/Effects/Fort_Effects/Textures/Tile/T_Hex_Noise}
  n8 = TextureCoordinate()
  n9 = Sine(None=Multiply(A=ComponentMask(None=@n8) {r=true}) {constB=0.5})
  Multiply(A=@n1.A, B=Multiply(A=Multiply(A=@n2, B=@n7.R), B=Power(Base=@n9)))
 ]
}
