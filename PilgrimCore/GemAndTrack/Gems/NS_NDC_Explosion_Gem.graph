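# Niagara explosion effect for gem hits on a track, combining flares, a smasher mesh, sparks and perfect-hit layers.
# All emitters spawn from a shared Data Channel, passing User.EmitterTrackIndex, with burst spawns disabled.
# Each particle reads position and color data from the channel; perfect-hit layers pick HDR colors by lane index.
# Perfect-hit glow, radial burst and sparkle lifetimes are multiplied by PerfectHitProbabilty, so a zero value collapses them.
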NiagaraSystem /PilgrimCore/GemAndTrack/Gems/NS_NDC_Explosion_Gem
dependencies {
 usedDataInterfaces [
  &NiagaraDataInterfaceCurve
  &NiagaraDataInterfaceArrayColor
  /Script/Niagara.NiagaraDataInterfaceDataChannelRead
 ]
 usedDynamicInputs []
 usedModules [
  &self:SystemSpawnScript.SystemScriptSource.SystemScriptGraph.NiagaraNodeAssignment_1.SetVariables_#1
  &self:SystemSpawnScript.SystemScriptSource.SystemScriptGraph.NiagaraNodeAssignment_2.SetVariables_#2
  &SystemState
 ]
}
emitters {
 FlareZAxis {
  emitterUpdateScript [
   &EmitterState()
   &SpawnBurst_Instantaneous(Spawn Count=1, Use Spawn Probability=false) {enabled=false}
   &self:SpawnFromNDC(Data Channel Reader:NiagaraDataInterfaceDataChannelRead=$System.Data Channel Reader, EmitterTrackIndex:NiagaraInt32=$User.EmitterTrackIndex) {module=SpawnFromNDC}
  ]
  particleSpawnScript [
   &self:ReadFromNDC(
    DataChannel interface:NiagaraDataInterfaceDataChannelRead=$System.Data Channel Reader
    Default Lane Colors:NiagaraDataInterfaceArrayColor=$System.Default Lane Colors
   ) {module=ReadFromNDC}
   &InitializeParticle(
    Color=$Output.ReadFromNDC.GemHitColor
    Lifetime=0.15
    Position=$Output.ReadFromNDC.Position
    Position Mode=::NewEnumerator1
    Sprite Size=<75 100>
    Sprite Size Mode=::NewEnumerator4
    UsePositionOffset=true
    Write Material Random=true
   )
   &DynamicMaterialParameters(
    Index 0 Param 1=&RandomRangeFloat(Minimum=0.4, Maximum=0.6)
    Index 0 Param 2=1
    Index 0 Param 3=1
    Index 0 Param 4=1
    Module.Param0WriteEnabled=true
    Module.Param1WriteEnabled=true
    Module.Param2WriteEnabled=true
    Module.Param3WriteEnabled=true
   )
  ]
  particleUpdateScript [
   &ParticleState()
   &ScaleColor(
    Scale Alpha=&Multiply_Float(
     A=&FNTimeOfDayScalar(Night Scale=0.6, Morning Scale=0.3, Evening Scale=0.4)
     B=&FloatFromCurve(FloatCurve=DataInterface() {propertyValues {Curve {keys [
          {interpMode=RCIM_Cubic, tangentMode=RCTM_User, time=0, value=0}
          {interpMode=RCIM_Cubic, tangentMode=RCTM_User, time=0.1, value=1}
          {interpMode=RCIM_Cubic, tangentMode=RCTM_User, time=0.5, value=1}
          {interpMode=RCIM_Cubic, tangentMode=RCTM_User, time=1, value=0}
         ]}}})
    )
    Scale RGB=&Multiply_VectorByFloat(Vector=<1 1 1>, Float=4)
    ScaleA=true
    ScaleRGB=true
   )
   &ScaleSpriteSize(Scale Factor=&Vector2DFromFloat(Value=&FloatFromCurve(FloatCurve=DataInterface() {propertyValues {Curve {keys [<at 0 0.5 RCIM_Cubic>,<at 1 1.25>]}}})))
   &SpriteFacingAndAlignment(Sprite Facing=<0 0 1>, WriteAlignment=false, WriteFacing=true)
  ]
  rendererClasses [&NiagaraSpriteRendererProperties]
  renderers [{rendererClass=&NiagaraSpriteRendererProperties}]
 }
 FlareZAxisStretched {
  emitterUpdateScript [
   &EmitterState()
   &SpawnBurst_Instantaneous(Spawn Count=1, Use Spawn Probability=false) {enabled=false}
   &self:SpawnFromNDC(Data Channel Reader:NiagaraDataInterfaceDataChannelRead=$System.Data Channel Reader, EmitterTrackIndex:NiagaraInt32=$User.EmitterTrackIndex) {module=SpawnFromNDC}
  ]
  particleSpawnScript [
   &self:ReadFromNDC(
    DataChannel interface:NiagaraDataInterfaceDataChannelRead=$System.Data Channel Reader
    Default Lane Colors:NiagaraDataInterfaceArrayColor=$System.Default Lane Colors
   ) {module=ReadFromNDC}
   &InitializeParticle(
    Color=&Multiply_LinearColorByFloat(
     Linear Color=$Output.ReadFromNDC.GemHitColor
     Float=&MakeCustomFloatFromBool(Bool=false, True Float=1, False Float=0.35)
    )
    Lifetime=0.15
    Position=$Output.ReadFromNDC.Position
    Position Mode=::NewEnumerator1
    Sprite Size=<10 25>
    Sprite Size Mode=::NewEnumerator4
    UsePositionOffset=true
    Write Material Random=true
   )
   &DynamicMaterialParameters(
    Index 0 Param 1=&RandomRangeFloat(Minimum=0.4, Maximum=0.6)
    Index 0 Param 2=1
    Index 0 Param 3=1
    Index 0 Param 4=1
    Module.Param0WriteEnabled=true
    Module.Param1WriteEnabled=true
    Module.Param2WriteEnabled=true
    Module.Param3WriteEnabled=true
   )
  ]
  particleUpdateScript [
   &ParticleState()
   &ScaleColor(Scale RGB=&Multiply_VectorByFloat(Vector=<1 1 1>, Float=8), ScaleA=true, ScaleRGB=true)
   &ScaleSpriteSize(Scale Factor=&Vector2DFromFloat(Value=&FloatFromCurve(FloatCurve=DataInterface() {propertyValues {Curve {keys [<at 0 0.5 RCIM_Cubic>,<at 1 1.25>]}}})))
   &SpriteFacingAndAlignment(Sprite Facing=<0 0 1>, WriteAlignment=false, WriteFacing=true)
  ]
  rendererClasses [&NiagaraSpriteRendererProperties]
  renderers [{rendererClass=&NiagaraSpriteRendererProperties}]
 }
 PerfectHitRadialExplode {
  emitterUpdateScript [
   &EmitterState(Life Cycle Mode=::NewEnumerator1, Loop Duration=0.2, UseLoopDelay=false)
   &SpawnBurst_Instantaneous(Use Spawn Probability=false, UseLoopCountLimit=false) {enabled=false}
   &self:SpawnFromNDC(Data Channel Reader:NiagaraDataInterfaceDataChannelRead=$System.Data Channel Reader, EmitterTrackIndex:NiagaraInt32=$User.EmitterTrackIndex) {module=SpawnFromNDC}
  ]
  particleSpawnScript [
   &self:ReadFromNDC(
    DataChannel interface:NiagaraDataInterfaceDataChannelRead=$System.Data Channel Reader
    Default Lane Colors:NiagaraDataInterfaceArrayColor=$System.Default Lane Colors
   ) {module=ReadFromNDC}
   &InitializeParticle(
    Color=&SelectLinearColorFromArray(
     Array Sampling Mode=::NewEnumerator1
     Direct Array Index=$Output.ReadFromNDC.LaneColor
     Color Selection Array=DataInterface() {propertyValues {ColorData [
        <rgba 50 17.5 1.25 1>
        <rgba 0.625 25 3.2004 1>
        <rgba 25 1.8383 0.625 1>
        <rgba 25 16.0672 0.625 1>
        <rgba 0.625 12.6645 25 1>
        <rgba 25 5.5697 0.625 1>
       ]}}
     Random Seed=?
    )
    Lifetime=&Multiply_Float(A=0.5, B=$Output.ReadFromNDC.PerfectHitProbabilty)
    Mesh Scale=<2.9 2.9 1>
    Mesh Scale Mode=::NewEnumerator4
    Position=$Output.ReadFromNDC.Position
    Position Mode=::NewEnumerator1
    Position Offset=<0 0 -0.75>
    UsePositionOffset=true
    Write Material Random=true
   )
  ]
  particleUpdateScript [
   &ParticleState()
   &ScaleMeshSize(Scale Factor=&MakeVector(
     X=&FloatFromCurve(FloatCurve=DataInterface() {propertyValues {Curve {keys [<at 0 1>,<at 1 1.5>]}}})
     Y=&FloatFromCurve(FloatCurve=DataInterface() {propertyValues {Curve {keys [<at 0 1>,<at 1 5>]}}})
     Z=1
    ))
   &SolveForcesAndVelocity(Clamp Velocity=false, Limit Acceleration=false)
   &ScaleColor(
    Scale Alpha=&FloatFromCurve(FloatCurve=DataInterface() {propertyValues {Curve {keys [
         {arriveTangent=9.0725, interpMode=RCIM_Cubic, leaveTangent=9.0725, tangentMode=RCTM_User, time=0, value=0}
         {arriveTangent=-2.5249, interpMode=RCIM_Cubic, leaveTangent=-2.5249, tangentMode=RCTM_User, time=0.1408, value=0.9847}
         {arriveTangent=-0.0405, interpMode=RCIM_Cubic, leaveTangent=-0.0405, tangentMode=RCTM_User, time=1, value=0}
        ]}}})
    Scale RGB=&VectorFromFloat(Value=&FloatFromCurve(FloatCurve=DataInterface() {propertyValues {Curve {keys [
          {arriveTangent=9.0725, interpMode=RCIM_Cubic, leaveTangent=9.0725, tangentMode=RCTM_User, time=0, value=0}
          {arriveTangent=-2.5249, interpMode=RCIM_Cubic, leaveTangent=-2.5249, tangentMode=RCTM_User, time=0.1408, value=0.9847}
          {arriveTangent=-0.0405, interpMode=RCIM_Cubic, leaveTangent=-0.0405, tangentMode=RCTM_User, time=1, value=0}
         ]}}}))
    ScaleA=true
    ScaleRGB=true
   )
   &DynamicMaterialParameters(
    Index 0 Param 1=&FloatFromCurve(FloatCurve=DataInterface() {propertyValues {Curve {keys [<at 0 2>,<at 1 1>]}}})
    Module.Index0Param0Write=true
    Module.Index0Param1Write=true
    Module.Index0Param2Write=true
    Module.Index0Param3Write=true
   )
  ]
  rendererClasses [&NiagaraMeshRendererProperties]
  renderers [{rendererClass=&NiagaraMeshRendererProperties, rendererIndex=0}]
 }
 PerfectHitSmasherGlowEdge {
  emitterUpdateScript [
   &EmitterState(Life Cycle Mode=::NewEnumerator1, Loop Duration=0.1, UseLoopDelay=false)
   &SpawnBurst_Instantaneous(Use Spawn Probability=false, UseLoopCountLimit=false) {enabled=false}
   &self:SpawnFromNDC(Data Channel Reader:NiagaraDataInterfaceDataChannelRead=$System.Data Channel Reader, EmitterTrackIndex:NiagaraInt32=$User.EmitterTrackIndex) {module=SpawnFromNDC}
  ]
  particleSpawnScript [
   &self:ReadFromNDC(
    DataChannel interface:NiagaraDataInterfaceDataChannelRead=$System.Data Channel Reader
    Default Lane Colors:NiagaraDataInterfaceArrayColor=$System.Default Lane Colors
   ) {module=ReadFromNDC}
   &InitializeParticle(
    Color=&SelectLinearColorFromArray(
     Array Sampling Mode=::NewEnumerator1
     Direct Array Index=$Output.ReadFromNDC.LaneColor
     Color Selection Array=DataInterface() {propertyValues {ColorData [
        <rgba 50 26.25 2.5 1>
        <rgba 2.5 50 7.5187 1>
        <rgba 50 4.8644 2.5 1>
        <rgba 50 32.5924 2.5 1>
        <rgba 2.5 25.9616 50 1>
        <rgba 50 12.1358 2.5 1>
       ]}}
     Random Seed=?
    )
    Lifetime=&Multiply_Float(A=0.25, B=$Output.ReadFromNDC.PerfectHitProbabilty)
    Mesh Scale=<2.9 2.9 1>
    Mesh Scale Mode=::NewEnumerator4
    Position=$Output.ReadFromNDC.Position
    Position Mode=::NewEnumerator1
    Position Offset=<0 0 -0.75>
    UsePositionOffset=true
    Write Material Random=true
   )
  ]
  particleUpdateScript [
   &ParticleState()
   &SolveForcesAndVelocity(Clamp Velocity=false, Limit Acceleration=false)
   &ScaleColor(
    Scale Alpha=&FloatFromCurve(FloatCurve=DataInterface() {propertyValues {Curve {keys [
         {arriveTangent=9.0725, interpMode=RCIM_Cubic, leaveTangent=9.0725, tangentMode=RCTM_User, time=0, value=0}
         {arriveTangent=-2.5249, interpMode=RCIM_Cubic, leaveTangent=-2.5249, tangentMode=RCTM_User, time=0.1408, value=0.9847}
         {arriveTangent=-0.0405, interpMode=RCIM_Cubic, leaveTangent=-0.0405, tangentMode=RCTM_User, time=1, value=0}
        ]}}})
    Scale RGB=&VectorFromFloat(Value=&FloatFromCurve(FloatCurve=DataInterface() {propertyValues {Curve {keys [
          {arriveTangent=9.0725, interpMode=RCIM_Cubic, leaveTangent=9.0725, tangentMode=RCTM_User, time=0, value=0}
          {arriveTangent=-2.5249, interpMode=RCIM_Cubic, leaveTangent=-2.5249, tangentMode=RCTM_User, time=0.1408, value=0.9847}
          {arriveTangent=-0.0405, interpMode=RCIM_Cubic, leaveTangent=-0.0405, tangentMode=RCTM_User, time=1, value=0}
         ]}}}))
    ScaleA=true
    ScaleRGB=true
   )
   &DynamicMaterialParameters(
    Index 0 Param 1=1.5
    Module.Index0Param0Write=true
    Module.Index0Param1Write=true
    Module.Index0Param2Write=true
    Module.Index0Param3Write=true
   )
  ]
  rendererClasses [&NiagaraMeshRendererProperties]
  renderers [{rendererClass=&NiagaraMeshRendererProperties, rendererIndex=0}]
 }
 PerfectHitSparkle002 {
  emitterUpdateScript [
   &EmitterState()
   &SpawnBurst_Instantaneous(Spawn Count=3, Use Spawn Probability=false, UseLoopCountLimit=false) {enabled=false}
   &self:SpawnFromNDC(
    Data Channel Reader:NiagaraDataInterfaceDataChannelRead=$System.Data Channel Reader
    EmitterTrackIndex:NiagaraInt32=$User.EmitterTrackIndex
    Spawn Count=3
   ) {module=SpawnFromNDC}
  ]
  particleSpawnScript [
   &self:ReadFromNDC(
    DataChannel interface:NiagaraDataInterfaceDataChannelRead=$System.Data Channel Reader
    Default Lane Colors:NiagaraDataInterfaceArrayColor=$System.Default Lane Colors
   ) {module=ReadFromNDC}
   &InitializeParticle(
    Color=&SelectLinearColorFromArray(
     Array Sampling Mode=::NewEnumerator1
     Direct Array Index=$Output.ReadFromNDC.LaneColor
     Color Selection Array=DataInterface() {propertyValues {ColorData [<rgba 6 3.15 0.3 1>, <rgba 0.3 6 0.9022 1>, <rgba 6 0.5837 0.3 1>, <rgba 6 3.9111 0.3 1>, <rgba 0.3 3.1154 6 1>, <rgba 6 1.4563 0.3 1>]}}
     Random Seed=?
    )
    Lifetime Max=&Multiply_Float(A=0.3, B=$Output.ReadFromNDC.PerfectHitProbabilty)
    Lifetime Min:NiagaraFloat=&Multiply_Float(A=0.25, B=$Output.ReadFromNDC.PerfectHitProbabilty)
    Lifetime Mode=::NewEnumerator1
    Position=$Output.ReadFromNDC.Position
    Position Mode=::NewEnumerator1
    Sprite Size Mode=::NewEnumerator3
    Uniform Sprite Size=20
    UsePositionOffset=true
    Write Material Random=true
   )
   &ShapeLocation(Box Size=<3 0.1 0.1>, MaskSpawnGroup=false, Override Seed=false, Shape Primitive=::NewEnumerator2, Surface Only=false, UseOwnerScale=true) {module=ShapeLocation001}
   &AddVelocity(Velocity=&RandomRangeVector(Minimum=<0 0 5>, Maximum=<0 0 15>)) {enabled=false}
   &AddVelocity(
    Coordinate Space=::Simulation
    Scale Added Velocity=$Output.ReadFromNDC.Scale
    Velocity=&RandomRangeVector(Minimum=<0 0 5>, Maximum=<0 100 5>)
   ) {module=AddVelocity001}
  ]
  particleUpdateScript [
   &Drag(Drag=&RandomRangeFloat(Minimum=2, Maximum=4), Use Linear Drag=true, Use Rotational Drag=false)
   &SolveForcesAndVelocity(Clamp Velocity=false, Limit Acceleration=false)
   &ParticleState()
   &ScaleSpriteSize(Scale Factor=&Vector2DFromFloat(Value=&FloatFromCurve(FloatCurve=DataInterface() {propertyValues {Curve {keys [
          {arriveTangent=10.3344, interpMode=RCIM_Cubic, leaveTangent=10.3344, tangentMode=RCTM_User, time=0, value=0.2}
          {interpMode=RCIM_Cubic, tangentMode=RCTM_User, time=0.1, value=1}
          {arriveTangent=-0.8819, interpMode=RCIM_Cubic, leaveTangent=-0.8819, tangentMode=RCTM_User, time=0.3, value=0.5}
          {arriveTangent=-0.3038, interpMode=RCIM_Cubic, leaveTangent=-0.3038, tangentMode=RCTM_User, time=1, value=0}
         ]}}})))
   &ScaleColor(
    Scale Alpha=&FloatFromCurve(FloatCurve=DataInterface() {propertyValues {Curve {keys [<at 0 0>,<at 0.2 1>]}}})
    Scale RGB=&VectorFromFloat(Value=&FloatFromCurve(FloatCurve=DataInterface() {propertyValues {Curve {keys [<at 0 1>,<at 1 0>]}}}, Scale Curve=1.5))
    ScaleA=true
    ScaleRGB=true
   ) {module=ScaleColor001}
  ]
  rendererClasses [&NiagaraSpriteRendererProperties]
  renderers [{rendererClass=&NiagaraSpriteRendererProperties}]
 }
 SmasherMesh {
  emitterUpdateScript [
   &EmitterState()
   &SpawnBurst_Instantaneous(Use Spawn Probability=false, UseLoopCountLimit=false) {enabled=false}
   &self:SpawnFromNDC(Data Channel Reader:NiagaraDataInterfaceDataChannelRead=$System.Data Channel Reader, EmitterTrackIndex:NiagaraInt32=$User.EmitterTrackIndex) {module=SpawnFromNDC}
  ]
  particleSpawnScript [
   &self:ReadFromNDC(
    DataChannel interface:NiagaraDataInterfaceDataChannelRead=$System.Data Channel Reader
    Default Lane Colors:NiagaraDataInterfaceArrayColor=$System.Default Lane Colors
   ) {module=ReadFromNDC}
   &InitializeParticle(
    Color=$Output.ReadFromNDC.GemHitColor
    Lifetime=0.2
    Mesh Scale=<3 3 2.9>
    Mesh Scale Mode=::NewEnumerator4
    Position=$Output.ReadFromNDC.Position
    Position Mode=::NewEnumerator1
    Position Offset=<0 0 -0.75>
    UsePositionOffset=true
    Write Material Random=true
   )
  ]
  particleUpdateScript [
   &ParticleState()
   &ScaleMeshSize(Scale Factor=&MakeVector(
     X=1
     Y=1
     Z=&FloatFromCurve(
      FloatCurve=DataInterface() {propertyValues {Curve {keys [
          {arriveTangent=9.0725, interpMode=RCIM_Cubic, leaveTangent=9.0725, tangentMode=RCTM_User, time=0, value=0}
          {arriveTangent=0.0462, interpMode=RCIM_Cubic, leaveTangent=0.0462, tangentMode=RCTM_User, time=0.025, value=1}
          {arriveTangent=-0.0405, interpMode=RCIM_Cubic, leaveTangent=-0.0405, tangentMode=RCTM_User, time=1, value=0.25}
         ]}}}
      Scale Curve=7
     )
    ))
   &ScaleColor(
    Scale Alpha=&FloatFromCurve(FloatCurve=DataInterface() {propertyValues {Curve {keys [{interpMode=RCIM_Cubic, tangentMode=RCTM_User, time=0, value=1}, {interpMode=RCIM_Cubic, tangentMode=RCTM_User, time=1, value=0}]}}})
    Scale RGB=&VectorFromFloat(Value=2)
    ScaleA=true
    ScaleRGB=true
   )
  ]
  rendererClasses [&NiagaraMeshRendererProperties]
  renderers [{rendererClass=&NiagaraMeshRendererProperties, rendererIndex=0}]
 }
 TechSparks {
  emitterUpdateScript [
   &EmitterState()
   &SpawnBurst_Instantaneous(Spawn Count=3, Use Spawn Probability=false, UseLoopCountLimit=false) {enabled=false}
   &self:SpawnFromNDC(
    Data Channel Reader:NiagaraDataInterfaceDataChannelRead=$System.Data Channel Reader
    EmitterTrackIndex:NiagaraInt32=$User.EmitterTrackIndex
    Spawn Count=3
   ) {module=SpawnFromNDC}
  ]
  particleSpawnScript [
   &self:ReadFromNDC(
    DataChannel interface:NiagaraDataInterfaceDataChannelRead=$System.Data Channel Reader
    Default Lane Colors:NiagaraDataInterfaceArrayColor=$System.Default Lane Colors
   ) {module=ReadFromNDC}
   &InitializeParticle(
    Color=$Output.ReadFromNDC.GemHitColor
    Lifetime Max=0.2
    Lifetime Min=0.15
    Lifetime Mode=::NewEnumerator1
    Position=$Output.ReadFromNDC.Position
    Position Mode=::NewEnumerator1
    Sprite Size Max=<0.95 5>
    Sprite Size Min=<0.65 2>
    Sprite Size Mode=::NewEnumerator2
    UsePositionOffset=true
    Write Material Random=true
   )
   &ShapeLocation(Box Size=<2 0.2 0.2>, MaskSpawnGroup=false, Override Seed=false, Shape Primitive=::NewEnumerator2, Surface Only=false, UseOwnerScale=true)
   &AddVelocity(
    Coordinate Space=::Simulation
    Scale Added Velocity=$Output.ReadFromNDC.Scale
    Velocity=&RandomRangeVector(Minimum=<0 0 15>, Maximum=<0 0 30>)
   ) {module=AddVelocity001}
   &DynamicMaterialParameters(
    Index 0 Param 1=0.8
    Module.Index0Param0Write=true
    Module.Index0Param1Write=true
    Module.Index0Param2Write=true
    Module.Index0Param3Write=true
   )
  ]
  particleUpdateScript [
   &Drag(Drag=&RandomRangeFloat(Minimum=2, Maximum=4), Use Linear Drag=true, Use Rotational Drag=false)
   &SolveForcesAndVelocity(Clamp Velocity=false, Limit Acceleration=false)
   &ParticleState()
   &ScaleSpriteSize(Scale Factor=&MakeVector2D(
     X=1
     Y=&FloatFromCurve(FloatCurve=DataInterface() {propertyValues {Curve {keys [<at 0.2 1>,<at 1 0.25>]}}})
    )) {module=ScaleSpriteSize002}
   &ScaleColor(
    Scale Alpha=&FloatFromCurve(FloatCurve=DataInterface() {propertyValues {Curve {keys [<at 0 0>,<at 0.2 1>]}}})
    Scale RGB=&VectorFromFloat(Value=&FloatFromCurve(FloatCurve=DataInterface() {propertyValues {Curve {keys [{interpMode=RCIM_Cubic, tangentMode=RCTM_User, time=0, value=40}, {interpMode=RCIM_Cubic, tangentMode=RCTM_User, time=1, value=1}]}}}))
    ScaleA=true
    ScaleRGB=true
   )
  ]
  rendererClasses [&NiagaraSpriteRendererProperties]
  renderers [{rendererClass=&NiagaraSpriteRendererProperties}]
 }
}
summary {
 systemName=NS_NDC_Explosion_Gem
 emitters [
  {emitterName=SmasherMesh, rendererClasses [&NiagaraMeshRendererProperties]}
  {emitterName=FlareZAxis, rendererClasses [&NiagaraSpriteRendererProperties]}
  {emitterName=FlareZAxisStretched, rendererClasses [&NiagaraSpriteRendererProperties]}
  {emitterName=PerfectHitSparkle002, rendererClasses [&NiagaraSpriteRendererProperties]}
  {emitterName=PerfectHitSmasherGlowEdge, rendererClasses [&NiagaraMeshRendererProperties]}
  {emitterName=PerfectHitRadialExplode, rendererClasses [&NiagaraMeshRendererProperties]}
  {emitterName=TechSparks, rendererClasses [&NiagaraSpriteRendererProperties]}
 ]
 userVariables [User.EmitterTrackIndex {type=NiagaraInt32}]
}
