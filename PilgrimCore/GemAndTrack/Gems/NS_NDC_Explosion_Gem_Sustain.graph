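# Niagara system for the sustained gem hit explosion on track lanes, driven by a Niagara Data Channel.
# Each emitter's SpawnFromNDC spawns at a rate using User.EmitterTrackIndex, replacing the disabled SpawnRate, and ReadFromNDC supplies position and gem hit color.
# Five short-lived layers form the effect: two flare sprites, a smasher mesh with edge glow, and dragged, shrinking sparks.
# The Z-axis flare's alpha also scales with time of day.
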NiagaraSystem /PilgrimCore/GemAndTrack/Gems/NS_NDC_Explosion_Gem_Sustain
dependencies {
 usedDataInterfaces [
  &NiagaraDataInterfaceCurve
  /Script/Niagara.NiagaraDataInterfaceDataChannelRead
  &NiagaraDataInterfaceArrayColor
 ]
 usedDynamicInputs []
 usedModules [
  &self:SystemSpawnScript.SystemScriptSource.SystemScriptGraph.NiagaraNodeAssignment_0.SetVariables_#1
  &self:SystemSpawnScript.SystemScriptSource.SystemScriptGraph.NiagaraNodeAssignment_2.SetVariables_#2
  &SystemState
 ]
}
emitters {
 FlareXAxisStretched {
  emitterUpdateScript [
   &EmitterState()
   &SpawnRate(SpawnRate=10, Use Spawn Probability=false) {enabled=false}
   &self:SpawnFromNDC(
    Data Channel Reader:NiagaraDataInterfaceDataChannelRead=$System.Data Channel Reader
    EmitterTrackIndex:NiagaraInt32=$User.EmitterTrackIndex
    Spawn Rate=10
   ) {module=SpawnFromNDC}
  ]
  particleSpawnScript [
   &self:ReadFromNDC(
    DataChannel interface:NiagaraDataInterfaceDataChannelRead=$System.Data Channel Reader
    Default Lane Colors:NiagaraDataInterfaceArrayColor=$System.Default Lane Colors
   ) {module=ReadFromNDC}
   &InitializeParticle(
    Color=$Output.ReadFromNDC.GemHitColor
    Lifetime=0.15
    Position=$Output.ReadFromNDC.Position
    Position Mode=::NewEnumerator1
    Position Offset=<0 1 0>
    Sprite Size=<10 15>
    Sprite Size Mode=::NewEnumerator4
    UsePositionOffset=true
    Write Material Random=true
   )
   &DynamicMaterialParameters(
    Index 0 Param 1=&RandomRangeFloat(Minimum=0.4, Maximum=0.6)
    Index 0 Param 2=1
    Index 0 Param 3=1
    Index 0 Param 4=1
    Module.Param0WriteEnabled=true
    Module.Param1WriteEnabled=true
    Module.Param2WriteEnabled=true
    Module.Param3WriteEnabled=true
   )
  ]
  particleUpdateScript [
   &ParticleState()
   &ScaleColor(Scale RGB=&Multiply_VectorByFloat(Vector=<1 1 1>, Float=25), ScaleA=true, ScaleRGB=true)
   &ScaleSpriteSize(Scale Factor=&Vector2DFromFloat(Value=&FloatFromCurve(FloatCurve=DataInterface() {propertyValues {Curve {keys [<at 0 0.5 RCIM_Cubic>,<at 1 1.25>]}}})))
   &SpriteFacingAndAlignment(Sprite Facing=<0 0 1>, WriteAlignment=false, WriteFacing=true)
  ]
  rendererClasses [&NiagaraSpriteRendererProperties]
  renderers [{rendererClass=&NiagaraSpriteRendererProperties}]
 }
 FlareZAxis {
  emitterUpdateScript [
   &EmitterState()
   &SpawnRate(SpawnRate=10, Use Spawn Probability=false) {enabled=false}
   &self:SpawnFromNDC(
    Data Channel Reader:NiagaraDataInterfaceDataChannelRead=$System.Data Channel Reader
    EmitterTrackIndex:NiagaraInt32=$User.EmitterTrackIndex
    Spawn Rate=10
   ) {module=SpawnFromNDC}
  ]
  particleSpawnScript [
   &self:ReadFromNDC(
    DataChannel interface:NiagaraDataInterfaceDataChannelRead=$System.Data Channel Reader
    Default Lane Colors:NiagaraDataInterfaceArrayColor=$System.Default Lane Colors
   ) {module=ReadFromNDC}
   &InitializeParticle(
    Color=$Output.ReadFromNDC.GemHitColor
    Lifetime=0.15
    Position=$Output.ReadFromNDC.Position
    Position Mode=::NewEnumerator1
    Sprite Size=<75 100>
    Sprite Size Mode=::NewEnumerator4
    UsePositionOffset=true
    Write Material Random=true
   )
   &DynamicMaterialParameters(
    Index 0 Param 1=&RandomRangeFloat(Minimum=0.4, Maximum=0.6)
    Index 0 Param 2=1
    Index 0 Param 3=1
    Index 0 Param 4=1
    Module.Param0WriteEnabled=true
    Module.Param1WriteEnabled=true
    Module.Param2WriteEnabled=true
    Module.Param3WriteEnabled=true
   )
  ]
  particleUpdateScript [
   &ParticleState()
   &ScaleColor(
    Scale Alpha=&Multiply_Float(
     A=&FNTimeOfDayScalar(Night Scale=0.6, Morning Scale=0.3, Evening Scale=0.4)
     B=&FloatFromCurve(FloatCurve=DataInterface() {propertyValues {Curve {keys [
          {interpMode=RCIM_Cubic, tangentMode=RCTM_User, time=0, value=0}
          {interpMode=RCIM_Cubic, tangentMode=RCTM_User, time=0.1, value=1}
          {interpMode=RCIM_Cubic, tangentMode=RCTM_User, time=0.5, value=1}
          {interpMode=RCIM_Cubic, tangentMode=RCTM_User, time=1, value=0}
         ]}}})
    )
    Scale RGB=&Multiply_VectorByFloat(Vector=<1 1 1>, Float=1)
    ScaleA=true
    ScaleRGB=true
   )
   &ScaleSpriteSize(Scale Factor=&Vector2DFromFloat(Value=&FloatFromCurve(FloatCurve=DataInterface() {propertyValues {Curve {keys [<at 0 0.5 RCIM_Cubic>,<at 1 1.25>]}}})))
   &SpriteFacingAndAlignment(Sprite Facing=<0 0 1>, WriteAlignment=false, WriteFacing=true)
  ]
  rendererClasses [&NiagaraSpriteRendererProperties]
  renderers [{rendererClass=&NiagaraSpriteRendererProperties}]
 }
 SmasherEdgeGlow {
  emitterUpdateScript [
   &EmitterState()
   &SpawnRate(SpawnRate=10, Use Spawn Probability=false) {enabled=false}
   &self:SpawnFromNDC(
    Data Channel Reader:NiagaraDataInterfaceDataChannelRead=$System.Data Channel Reader
    EmitterTrackIndex:NiagaraInt32=$User.EmitterTrackIndex
    Spawn Rate=10
   ) {module=SpawnFromNDC}
  ]
  particleSpawnScript [
   &self:ReadFromNDC(
    DataChannel interface:NiagaraDataInterfaceDataChannelRead=$System.Data Channel Reader
    Default Lane Colors:NiagaraDataInterfaceArrayColor=$System.Default Lane Colors
   ) {module=ReadFromNDC}
   &InitializeParticle(
    Color=$Output.ReadFromNDC.GemHitColor
    Lifetime=0.25
    Mesh Scale=<2.9 2.9 1>
    Mesh Scale Mode=::NewEnumerator4
    Position=$Output.ReadFromNDC.Position
    Position Mode=::NewEnumerator1
    Position Offset=<0 0 -0.75>
    UsePositionOffset=true
    Write Material Random=true
   )
  ]
  particleUpdateScript [
   &ParticleState()
   &SolveForcesAndVelocity(Clamp Velocity=false, Limit Acceleration=false)
   &ScaleColor(
    Scale Alpha=&FloatFromCurve(FloatCurve=DataInterface() {propertyValues {Curve {keys [
         {arriveTangent=9.0725, interpMode=RCIM_Cubic, leaveTangent=9.0725, tangentMode=RCTM_User, time=0, value=0}
         {arriveTangent=-2.5249, interpMode=RCIM_Cubic, leaveTangent=-2.5249, tangentMode=RCTM_User, time=0.1408, value=0.9847}
         {arriveTangent=-0.0405, interpMode=RCIM_Cubic, leaveTangent=-0.0405, tangentMode=RCTM_User, time=1, value=0}
        ]}}})
    Scale RGB=&VectorFromFloat(Value=&FloatFromCurve(
      FloatCurve=DataInterface() {propertyValues {Curve {keys [
          {arriveTangent=9.0725, interpMode=RCIM_Cubic, leaveTangent=9.0725, tangentMode=RCTM_User, time=0, value=0}
          {arriveTangent=-2.5249, interpMode=RCIM_Cubic, leaveTangent=-2.5249, tangentMode=RCTM_User, time=0.1408, value=0.9847}
          {arriveTangent=-0.0405, interpMode=RCIM_Cubic, leaveTangent=-0.0405, tangentMode=RCTM_User, time=1, value=0}
         ]}}}
      Scale Curve=10
     ))
    ScaleA=true
    ScaleRGB=true
   )
   &DynamicMaterialParameters(
    Index 0 Param 1=1.5
    Module.Index0Param0Write=true
    Module.Index0Param1Write=true
    Module.Index0Param2Write=true
    Module.Index0Param3Write=true
   )
  ]
  rendererClasses [&NiagaraMeshRendererProperties]
  renderers [{rendererClass=&NiagaraMeshRendererProperties, rendererIndex=0}]
 }
 SmasherMesh {
  emitterUpdateScript [
   &EmitterState()
   &SpawnRate(SpawnRate=15, Use Spawn Probability=false) {enabled=false}
   &self:SpawnFromNDC(
    Data Channel Reader:NiagaraDataInterfaceDataChannelRead=$System.Data Channel Reader
    EmitterTrackIndex:NiagaraInt32=$User.EmitterTrackIndex
    Spawn Rate=15
   ) {module=SpawnFromNDC}
  ]
  particleSpawnScript [
   &self:ReadFromNDC(
    DataChannel interface:NiagaraDataInterfaceDataChannelRead=$System.Data Channel Reader
    Default Lane Colors:NiagaraDataInterfaceArrayColor=$System.Default Lane Colors
   ) {module=ReadFromNDC}
   &InitializeParticle(
    Color=$Output.ReadFromNDC.GemHitColor
    Lifetime=0.1
    Mesh Scale=<3 3 13>
    Mesh Scale Mode=::NewEnumerator4
    Position=$Output.ReadFromNDC.Position
    Position Mode=::NewEnumerator1
    Position Offset=<0 0 -0.75>
    UsePositionOffset=true
    Write Material Random=true
   )
  ]
  particleUpdateScript [
   &ParticleState()
   &ScaleColor(Scale Alpha=0.5, Scale RGB=&VectorFromFloat(Value=0.5), ScaleA=true, ScaleRGB=true)
  ]
  rendererClasses [&NiagaraMeshRendererProperties]
  renderers [{rendererClass=&NiagaraMeshRendererProperties, rendererIndex=0}]
 }
 TechSparks003 {
  emitterUpdateScript [
   &EmitterState()
   &SpawnRate(SpawnRate=15, Use Spawn Probability=false) {enabled=false}
   &self:SpawnFromNDC(
    Data Channel Reader:NiagaraDataInterfaceDataChannelRead=$System.Data Channel Reader
    EmitterTrackIndex:NiagaraInt32=$User.EmitterTrackIndex
    Spawn Rate=15
   ) {module=SpawnFromNDC}
  ]
  particleSpawnScript [
   &self:ReadFromNDC(
    DataChannel interface:NiagaraDataInterfaceDataChannelRead=$System.Data Channel Reader
    Default Lane Colors:NiagaraDataInterfaceArrayColor=$System.Default Lane Colors
   ) {module=ReadFromNDC}
   &InitializeParticle(
    Color=$Output.ReadFromNDC.GemHitColor
    Lifetime Max=0.15
    Lifetime Min=0.1
    Lifetime Mode=::NewEnumerator1
    Position=$Output.ReadFromNDC.Position
    Position Mode=::NewEnumerator1
    Position Offset=<0 1 0>
    Sprite Size Max=<0.95 10>
    Sprite Size Min=<0.25 10>
    Sprite Size Mode=::NewEnumerator2
    UsePositionOffset=true
    Write Material Random=true
   )
   &ShapeLocation(Box Size=<0 0.2 0.2>, MaskSpawnGroup=false, Override Seed=false, Shape Primitive=::NewEnumerator2, Surface Only=false, UseOwnerScale=true)
   &AddVelocity(
    Coordinate Space=::Simulation
    Scale Added Velocity=$Output.ReadFromNDC.Scale
    Velocity=&RandomRangeVector(Minimum=<-50 -100 0>, Maximum=<50 -75 0>)
   ) {module=AddVelocity001}
   &DynamicMaterialParameters(
    Index 0 Param 1=0.8
    Module.Index0Param0Write=true
    Module.Index0Param1Write=true
    Module.Index0Param2Write=true
    Module.Index0Param3Write=true
   )
  ]
  particleUpdateScript [
   &Drag(Drag=7, Use Linear Drag=true, Use Rotational Drag=false)
   &SolveForcesAndVelocity(Clamp Velocity=false, Limit Acceleration=false)
   &ParticleState()
   &ScaleSpriteSize(Scale Factor=&MakeVector2D(
     X=1
     Y=&FloatFromCurve(FloatCurve=DataInterface() {propertyValues {Curve {keys [<at 0.2 1>,<at 1 0.25>]}}})
    )) {module=ScaleSpriteSize002}
   &ScaleColor(
    Scale Alpha=&FloatFromCurve(FloatCurve=DataInterface() {propertyValues {Curve {keys [<at 0 0>,<at 0.2 1>]}}})
    Scale RGB=&VectorFromFloat(Value=&FloatFromCurve(FloatCurve=DataInterface() {propertyValues {Curve {keys [{interpMode=RCIM_Cubic, tangentMode=RCTM_User, time=0, value=40}, {interpMode=RCIM_Cubic, tangentMode=RCTM_User, time=1, value=1}]}}}))
    ScaleA=true
    ScaleRGB=true
   )
  ]
  rendererClasses [&NiagaraSpriteRendererProperties]
  renderers [{rendererClass=&NiagaraSpriteRendererProperties}]
 }
}
summary {
 systemName=NS_NDC_Explosion_Gem_Sustain
 emitters [
  {emitterName=FlareZAxis, rendererClasses [&NiagaraSpriteRendererProperties]}
  {emitterName=SmasherMesh, rendererClasses [&NiagaraMeshRendererProperties]}
  {emitterName=FlareXAxisStretched, rendererClasses [&NiagaraSpriteRendererProperties]}
  {emitterName=SmasherEdgeGlow, rendererClasses [&NiagaraMeshRendererProperties]}
  {emitterName=TechSparks003, rendererClasses [&NiagaraSpriteRendererProperties]}
 ]
 userVariables [User.EmitterTrackIndex {type=NiagaraInt32}]
}
